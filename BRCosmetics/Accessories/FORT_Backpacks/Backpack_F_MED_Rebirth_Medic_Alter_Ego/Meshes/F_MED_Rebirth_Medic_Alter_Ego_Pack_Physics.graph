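# Physics asset for the Rebirth Medic Alter Ego backpack cosmetic, defining physics bodies for the pack and its front and back cords.
# Each of the three bones (pack, dyn_cord_front, dyn_cord_back) carries a single box shape.
# Both cord bodies are constrained directly to the pack with identical angular limits, allowing asymmetric swing and only slight twist.
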PhysicsAsset /BRCosmetics/Accessories/FORT_Backpacks/Backpack_F_MED_Rebirth_Medic_Alter_Ego/Meshes/F_MED_Rebirth_Medic_Alter_Ego_Pack_Physics
bodies {
 dyn_cord_back [{shapeType=Box, extentX=3.4626, extentY=3.5169, extentZ=1.3076, center=<0.8437 0 0>}]
 dyn_cord_front [{shapeType=Box, extentX=3.4626, extentY=3.5169, extentZ=1.3076, center=<0.6528 0 0>}]
 pack [{shapeType=Box, extentX=2.5, extentY=2, extentZ=2}]
}
constraints [
 {bone1Name=dyn_cord_front, bone2Name=pack, swing1LimitDegrees=25, swing2LimitDegrees=45, twistLimitDegrees=5}
 {bone1Name=dyn_cord_back, bone2Name=pack, swing1LimitDegrees=25, swing2LimitDegrees=45, twistLimitDegrees=5}
]
