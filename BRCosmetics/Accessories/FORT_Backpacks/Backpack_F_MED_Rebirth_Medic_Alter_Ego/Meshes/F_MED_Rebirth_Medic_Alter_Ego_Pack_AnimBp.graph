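# Animation Blueprint for the Rebirth Medic Alter Ego backpack cosmetic, adding physics motion and a root adjustment to its input pose.
# A RigidBody node simulates the pack's physics asset relative to the root bone with overridden gravity, weighted by SkydiveDynamicsAlpha.
# A Modify Bone node then additively shifts and rolls the root, weighted by BackpackTransformAlpha.
# The event graph is empty, and neither alpha variable is declared here.
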AnimBlueprint /BRCosmetics/Accessories/FORT_Backpacks/Backpack_F_MED_Rebirth_Medic_Alter_Ego/Meshes/F_MED_Rebirth_Medic_Alter_Ego_Pack_AnimBp
target_skeleton=&/F_MED_Rebirth_Medic_Alter_Ego_Pack_Skeleton
graphs [
 AnimGraph {nodes [
   AnimGraphNode_SubInput_0 = Input Pose()
   n1 = RigidBody(ComponentPose=Local To Component(LocalPose=@AnimGraphNode_SubInput_0), Alpha=Get SkydiveDynamicsAlpha() -> (SkydiveDynamicsAlpha:float)) {
    bClampLinearTranslationLimitToRefPose=true
    bOverrideWorldGravity=true
    overridePhysicsAsset=&/F_MED_Rebirth_Medic_Alter_Ego_Pack_Physics
    simulationSpace=BaseBoneSpace
    alphaBoolBlend {blendOption=Linear}
    baseBoneRef {boneName=root}
    componentAppliedLinearAccClamp=<300 300 300>
    componentLinearAccScale=<0.25 0.25 0.25>
    componentLinearVelScale=<0.25 0.25 0.25>
    overrideWorldGravity {z=-980}
   }
   n2 = "Transform (Modify) Bone - Bone: root"(ComponentPose=@n1, Alpha=Get BackpackTransformAlpha() -> (BackpackTransformAlpha:float), Translation="0,-3,0", Rotation="0,0,20") {
    rotationMode=BMM_Additive
    rotationSpace=BCS_BoneSpace
    translationMode=BMM_Additive
    translationSpace=BCS_BoneSpace
    alphaBoolBlend {blendOption=Linear}
    boneToModify {boneName=root}
    rotation {roll=20}
    translation {y=-3}
   }
   Output Pose(Result=Component To Local(ComponentPose=@n2)) {layerGroup=None}
  ]}
 EventGraph {links [], nodes []}
]
variables []
